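# Season 37 Battle Royale job board conversation offering Supply, Vehicle and Armory Jackpot briefings, one per entry point.
# Each entry offers a requirement-gated quest grant and Cancel, plus shared "cannot use" choices for active quests or used boards.
# Supply and Vehicle grants spawn actors and deactivate the board; Armory spawns tiered loot without deactivation or a prior-grant check.
# Vehicle links refusals from its entry node, the others from their Speech nodes.
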ConversationDatabase /JobBoard_BR_S37/Conversations/Conversation_JobBoard_S37
entry_points [
 {tag=Conversation.Entry.JobBoard.SupplyJackpot, destinations [#1]}
 {tag=Conversation.Entry.JobBoard.VehicleJackpot, destinations [#2]}
 {tag=Conversation.Entry.JobBoard.ArmoryJackpot, destinations [#3]}
]
nodes [
 ConversationEntryPointNode_5 {
  id=#3
  class=ConversationEntryPointNode
  outputs [#4]
  properties {entryTag {tagName=Conversation.Entry.JobBoard.ArmoryJackpot}}
 }
 Speech_C_12 {
  id=#4
  class=Speech_C
  outputs [
   #5
   #6
   #7
   #8
   #9
   #10
  ]
  properties {nodeName=""}
 }
 ShowUserDataDrivenCannotUseReason_C_5 {
  id=#10
  class=ShowUserDataDrivenCannotUseReason_C
  outputs [#11]
  properties {
   cannotUseReason=UrgentQuestAlreadyActive
   hoverChoiceSound=/JobBoard/Sounds/JobBoard_ShadowBriefing_NotAvailable_Cue
   serviceBriefCollectionKeyOveride=Service.JobBoard.DisplayCannotUseReason
   serviceBriefConfigCollection=/JobBoard/Conversations/JobBoardServiceBriefCollectionConfig
   softTaskIcon=&Cancel_Normal
  }
  sub_nodes [
   ConversationChoiceNode_3 {
    class=ConversationChoiceNode
    properties {defaultChoiceDisplayText=URGENT QUEST ALREADY ACTIVE, choiceTags {gameplayTags [{tagName=Conversation.Choice.Category.Back}]}}
   }
   FortConversationRequirement_ControllerRequirement_15 {
    class=FortConversationRequirement_ControllerRequirement
    properties {
     nodeName=Require: Has GameplayTag
     requirement {class=FortControllerRequirement_HasOwnedTag, properties {ownedTag {tagName=Granted.Athena.UrgentQuest.Active}}}
    }
   }
  ]
 }
 Continue_C_8 {id=#11, class=Continue_C}
 ShowUserDataDrivenCannotUseReason_C_1 {
  id=#9
  class=ShowUserDataDrivenCannotUseReason_C
  outputs [#12]
  properties {
   briefWidgetClassOverride=/FortniteConversationUI/Conversation/FortDataDrivenServiceBrief.FortDataDrivenServiceBrief_C
   cannotUseReason=NoBriefingsAvailable
   hoverChoiceSound=&JobBoard_ShadowBriefing_NotAvailable_Cue
   serviceBriefCollectionKeyOveride=Service.JobBoard.DisplayCannotUseReason
   serviceBriefConfigCollection=&JobBoardServiceBriefCollectionConfig
   softTaskIcon=&Cancel_Normal
  }
  sub_nodes [
   ConversationChoiceNode_25 {
    class=ConversationChoiceNode
    properties {defaultChoiceDisplayText=NO BRIEFINGS AVAILABLE, choiceTags {gameplayTags [{tagName=Conversation.Choice.Category.Back}]}}
   }
   FortConversationRequirement_ControllerRequirement_54 {
    class=FortConversationRequirement_ControllerRequirement
    properties {
     nodeName=Require: Has Recorded Conversation Event With NPC
     requirement {
      class=FortControllerRequirement_HasRecordedConversationEventWithNPC
      properties {queryMethod=CheckAgainstCurrentConversationParticipant, queryEventTag {tagName=Conversation.Event.GrantedQuest}}
     }
    }
   }
  ]
 }
 Continue_C_7 {id=#12, class=Continue_C}
 ShowUserDataDrivenCannotUseReason_C_0 {
  id=#8
  class=ShowUserDataDrivenCannotUseReason_C
  outputs [#13]
  properties {
   cannotUseReason=BriefingAlreadyActive
   hoverChoiceSound=&JobBoard_ShadowBriefing_NotAvailable_Cue
   serviceBriefCollectionKeyOveride=Service.JobBoard.DisplayCannotUseReason
   serviceBriefConfigCollection=&JobBoardServiceBriefCollectionConfig
   softTaskIcon=&Cancel_Normal
  }
  sub_nodes [
   ConversationChoiceNode_16 {
    class=ConversationChoiceNode
    properties {defaultChoiceDisplayText=BRIEFING ALREADY ACTIVE, choiceTags {gameplayTags [{tagName=Conversation.Choice.Category.Back}]}}
   }
   FortConversationRequirement_ControllerRequirement_26 {
    class=FortConversationRequirement_ControllerRequirement
    properties {
     nodeName=Require: Has GameplayTag
     requirement {class=FortControllerRequirement_HasOwnedTag, properties {ownedTag {tagName=Granted.Athena.BountyBoard.QuestActive}}}
    }
   }
  ]
 }
 Continue_C_9 {id=#13, class=Continue_C}
 ShowUserDataDrivenCannotUseReason_C_7 {
  id=#7
  class=ShowUserDataDrivenCannotUseReason_C
  outputs [#14]
  properties {
   cannotUseReason=BriefingAlreadyActive
   hoverChoiceSound=&JobBoard_ShadowBriefing_NotAvailable_Cue
   serviceBriefCollectionKeyOveride=Service.JobBoard.DisplayCannotUseReason
   serviceBriefConfigCollection=&JobBoardServiceBriefCollectionConfig
   softTaskIcon=&Cancel_Normal
  }
  sub_nodes [
   ConversationChoiceNode_21 {
    class=ConversationChoiceNode
    properties {defaultChoiceDisplayText=BRIEFING ALREADY ACTIVE, choiceTags {gameplayTags [{tagName=Conversation.Choice.Category.Back}]}}
   }
   FortConversationRequirement_ControllerRequirement_48 {
    class=FortConversationRequirement_ControllerRequirement
    properties {
     nodeName=Require: Has GameplayTag
     requirement {class=FortControllerRequirement_HasOwnedTag, properties {ownedTag {tagName=Granted.Athena.JobBoardExperimental.QuestActive}}}
    }
   }
  ]
 }
 Continue_C_5 {id=#14, class=Continue_C}
 Continue_C_1 {
  id=#6
  class=Continue_C
  properties {softTaskIcon=&Cancel_Normal}
  sub_nodes [
   ConversationChoiceNode_10 {
    class=ConversationChoiceNode
    properties {defaultChoiceDisplayText=CANCEL, choiceTags {gameplayTags [{tagName=Conversation.Choice.CustomSortPriority.Highest}]}}
   }
   FortConversationRequirement_ControllerRequirement_23 {
    class=FortConversationRequirement_ControllerRequirement
    properties {
     nodeName=Require: Has GameplayTag
     requirement {
      class=FortControllerRequirement_HasOwnedTag
      properties {bInverseCondition=true, ownedTag {tagName=Granted.Athena.JobBoardExperimental.QuestActive}}
     }
    }
   }
   FortConversationRequirement_ControllerRequirement_9 {
    class=FortConversationRequirement_ControllerRequirement
    properties {
     nodeName=Require: Has GameplayTag
     requirement {class=FortControllerRequirement_HasOwnedTag, properties {bInverseCondition=true, ownedTag {tagName=Granted.Athena.BountyBoard.QuestActive}}}
    }
   }
   FortConversationRequirement_ControllerRequirement_8 {
    class=FortConversationRequirement_ControllerRequirement
    properties {
     nodeName=Require: Has GameplayTag
     requirement {class=FortControllerRequirement_HasOwnedTag, properties {bInverseCondition=true, ownedTag {tagName=Granted.Athena.UrgentQuest.Active}}}
    }
   }
  ]
 }
 GrantSlottedQuest_C_8 {
  id=#5
  class=GrantSlottedQuest_C
  outputs []
  properties {
   bSharedQuestLimitsEnforced=false
   bShouldOverrideSpeakerText=false
   briefWidgetClassOverride=/JobBoard_BR_S37/Widgets/FortWeaponJackpotJobBrief.FortWeaponJackpotJobBrief_C
   choiceSound=null
   softTaskIcon=/Game/UI/Foundation/Textures/Icons/Checks/T-Icon-Check-128
   optionalFilterCategoryTag {tagName=Tandem.Service.JobType.S37.ArmoryJackpot}
  }
  sub_nodes [
   ConversationChoiceNode_24 {class=ConversationChoiceNode, properties {defaultChoiceDisplayText=HUNTER JACKPOT}}
   FortConversationRequirement_HasService_9 {
    class=FortConversationRequirement_HasService
    properties {bRequireExactMatch=false, serviceTag {tagName=Tandem.Service.JobType.S37.ArmoryJackpot}}
   }
   FortConversationRequirement_ControllerRequirement_51 {
    class=FortConversationRequirement_ControllerRequirement
    properties {
     nodeName=Require: Has Urgent Quest
     requirement {
      class=FortControllerRequirement_HasUrgentQuest
      properties {
       bInverseCondition=true
       bRequireExactMatch=false
       acceptedUrgentEventTags {gameplayTags [{tagName=SpecialEvent.HuntingPlayer.IsHunter}, {tagName=SpecialEvent.JobBoard.Experimental}]}
      }
     }
    }
   }
   FortConversationRequirement_ControllerRequirement_52 {
    class=FortConversationRequirement_ControllerRequirement
    properties {
     nodeName=Require: Has GameplayTag
     requirement {
      class=FortControllerRequirement_HasOwnedTag
      properties {bInverseCondition=true, ownedTag {tagName=Granted.Athena.JobBoardExperimental.QuestActive}}
     }
    }
   }
   FortConversationRequirement_ControllerRequirement_53 {
    class=FortConversationRequirement_ControllerRequirement
    properties {
     nodeName=Require: Has GameplayTag
     requirement {class=FortControllerRequirement_HasOwnedTag, properties {bInverseCondition=true, ownedTag {tagName=Granted.Athena.BountyBoard.QuestActive}}}
    }
   }
   SideEffect_PlayServiceUsedEffects_C_9 {class=SideEffect_PlayServiceUsedEffects_C}
   SideEffect_Server_RecordEventToHistoryData_C_9 {class=SideEffect_Server_RecordEventToHistoryData_C, properties {event Type {tagName=Conversation.Event.GrantedQuest}}}
   SideEffect_Server_TrackJobBoardUser_C_9 {class=SideEffect_Server_TrackJobBoardUser_C}
   SideEffect_Server_JobBoardBriefing_SpawnLoot_C_1 {class=SideEffect_Server_JobBoardBriefing_SpawnLoot_C, properties {lootTierGroupName=LTG_ArmoryJackpot}}
   FortConversationRequirement_ControllerRequirement_14 {
    class=FortConversationRequirement_ControllerRequirement
    properties {
     nodeName=Require: Has GameplayTag
     requirement {class=FortControllerRequirement_HasOwnedTag, properties {bInverseCondition=true, ownedTag {tagName=Granted.Athena.UrgentQuest.Active}}}
    }
   }
   Requirement_JobBoardBriefing_CanGiveLootForTier_C_11 {class=Requirement_JobBoardBriefing_CanGiveLootForTier_C, properties {lootTierGroupName=LTG_ArmoryJackpot}}
  ]
 }
 ConversationEntryPointNode_0 {
  id=#2
  class=ConversationEntryPointNode
  outputs [
   #15
   #7
   #8
   #9
   #10
  ]
  properties {entryTag {tagName=Conversation.Entry.JobBoard.VehicleJackpot}}
 }
 Speech_C_1 {
  id=#15
  class=Speech_C
  outputs [#16, #17]
  properties {nodeName=""}
 }
 Continue_C_0 {
  id=#17
  class=Continue_C
  properties {softTaskIcon=&Cancel_Normal}
  sub_nodes [
   ConversationChoiceNode_2 {
    class=ConversationChoiceNode
    properties {defaultChoiceDisplayText=CANCEL, choiceTags {gameplayTags [{tagName=Conversation.Choice.CustomSortPriority.Highest}]}}
   }
   FortConversationRequirement_ControllerRequirement_6 {
    class=FortConversationRequirement_ControllerRequirement
    properties {
     nodeName=Require: Has GameplayTag
     requirement {
      class=FortControllerRequirement_HasOwnedTag
      properties {bInverseCondition=true, ownedTag {tagName=Granted.Athena.JobBoardExperimental.QuestActive}}
     }
    }
   }
   FortConversationRequirement_ControllerRequirement_7 {
    class=FortConversationRequirement_ControllerRequirement
    properties {
     nodeName=Require: Has GameplayTag
     requirement {class=FortControllerRequirement_HasOwnedTag, properties {bInverseCondition=true, ownedTag {tagName=Granted.Athena.BountyBoard.QuestActive}}}
    }
   }
   FortConversationRequirement_ControllerRequirement_11 {
    class=FortConversationRequirement_ControllerRequirement
    properties {
     nodeName=Require: Has GameplayTag
     requirement {class=FortControllerRequirement_HasOwnedTag, properties {bInverseCondition=true, ownedTag {tagName=Granted.Athena.UrgentQuest.Active}}}
    }
   }
  ]
 }
 GrantSlottedQuest_C_0 {
  id=#16
  class=GrantSlottedQuest_C
  outputs []
  properties {
   bSharedQuestLimitsEnforced=false
   bShouldOverrideSpeakerText=false
   briefWidgetClassOverride=/JobBoard_BR_S37/Widgets/FortVehicleJackpotJobBrief.FortVehicleJackpotJobBrief_C
   choiceSound=null
   softTaskIcon=&T-Icon-Check-128
   optionalFilterCategoryTag {tagName=Tandem.Service.JobType.S37.VehicleJackpot}
  }
  sub_nodes [
   ConversationChoiceNode_4 {class=ConversationChoiceNode, properties {defaultChoiceDisplayText=VEHICLE JACKPOT BRIEFING}}
   SideEffect_PlayServiceUsedEffects_C_2 {class=SideEffect_PlayServiceUsedEffects_C}
   SideEffect_Server_RecordEventToHistoryData_C_2 {class=SideEffect_Server_RecordEventToHistoryData_C, properties {event Type {tagName=Conversation.Event.GrantedQuest}}}
   SideEffect_Server_TrackJobBoardUser_C_3 {class=SideEffect_Server_TrackJobBoardUser_C}
   SideEffect_Server_JobBoardExperimental_SpawnActorOfClass_C_0 {
    class=SideEffect_Server_JobBoardBriefing_SpawnActorOfClass_C
    properties {spawnerClass=/JobBoard_BR_S37/Blueprints/B_Athena_JobBoardBriefing_RiftVehicleJackpot_Spawner.B_Athena_JobBoardBriefing_RiftVehicleJackpot_Spawner_C}
   }
   SideEffect_Server_JobBoardExperimental_ApplyInstigatorGE_C_2 {class=SideEffect_Server_JobBoardBriefing_ApplyInstigatorGE_C}
   SideEffect_Server_JobBoardExperimental_DestroyOnUse_C_2 {class=SideEffect_Server_JobBoardBriefing_DeactivateOnUse_C}
   FortConversationRequirement_ControllerRequirement_17 {
    class=FortConversationRequirement_ControllerRequirement
    properties {
     nodeName=Require: Has Recorded Conversation Event With NPC
     requirement {
      class=FortControllerRequirement_HasRecordedConversationEventWithNPC
      properties {bInverseCondition=true, queryMethod=CheckAgainstCurrentConversationParticipant, queryEventTag {tagName=Conversation.Event.GrantedQuest}}
     }
    }
   }
   FortConversationRequirement_ControllerRequirement_18 {
    class=FortConversationRequirement_ControllerRequirement
    properties {
     nodeName=Require: Has Urgent Quest
     requirement {
      class=FortControllerRequirement_HasUrgentQuest
      properties {
       bInverseCondition=true
       bRequireExactMatch=false
       acceptedUrgentEventTags {gameplayTags [{tagName=SpecialEvent.HuntingPlayer.IsHunter}, {tagName=SpecialEvent.JobBoard.Experimental}]}
      }
     }
    }
   }
   FortConversationRequirement_HasService_1 {class=FortConversationRequirement_HasService, properties {serviceTag {tagName=Tandem.Service.JobType.S37.VehicleJackpot}}}
   FortConversationRequirement_ControllerRequirement_21 {
    class=FortConversationRequirement_ControllerRequirement
    properties {
     nodeName=Require: Has GameplayTag
     requirement {
      class=FortControllerRequirement_HasOwnedTag
      properties {bInverseCondition=true, ownedTag {tagName=Granted.Athena.JobBoardExperimental.QuestActive}}
     }
    }
   }
   FortConversationRequirement_ControllerRequirement_28 {
    class=FortConversationRequirement_ControllerRequirement
    properties {
     nodeName=Require: Has GameplayTag
     requirement {class=FortControllerRequirement_HasOwnedTag, properties {bInverseCondition=true, ownedTag {tagName=Granted.Athena.BountyBoard.QuestActive}}}
    }
   }
   FortConversationRequirement_ControllerRequirement_10 {
    class=FortConversationRequirement_ControllerRequirement
    properties {
     nodeName=Require: Has GameplayTag
     requirement {class=FortControllerRequirement_HasOwnedTag, properties {bInverseCondition=true, ownedTag {tagName=Granted.Athena.UrgentQuest.Active}}}
    }
   }
  ]
 }
 ConversationEntryPointNode_4 {
  id=#1
  class=ConversationEntryPointNode
  outputs [#18]
  properties {entryTag {tagName=Conversation.Entry.JobBoard.SupplyJackpot}}
 }
 Speech_C_0 {
  id=#18
  class=Speech_C
  outputs [
   #19
   #20
   #7
   #8
   #9
   #10
  ]
  properties {nodeName=""}
 }
 Continue_C_4 {
  id=#20
  class=Continue_C
  properties {softTaskIcon=&Cancel_Normal}
  sub_nodes [
   ConversationChoiceNode_1 {
    class=ConversationChoiceNode
    properties {defaultChoiceDisplayText=CANCEL, choiceTags {gameplayTags [{tagName=Conversation.Choice.CustomSortPriority.Highest}]}}
   }
   FortConversationRequirement_ControllerRequirement_4 {
    class=FortConversationRequirement_ControllerRequirement
    properties {
     nodeName=Require: Has GameplayTag
     requirement {
      class=FortControllerRequirement_HasOwnedTag
      properties {bInverseCondition=true, ownedTag {tagName=Granted.Athena.JobBoardExperimental.QuestActive}}
     }
    }
   }
   FortConversationRequirement_ControllerRequirement_5 {
    class=FortConversationRequirement_ControllerRequirement
    properties {
     nodeName=Require: Has GameplayTag
     requirement {class=FortControllerRequirement_HasOwnedTag, properties {bInverseCondition=true, ownedTag {tagName=Granted.Athena.BountyBoard.QuestActive}}}
    }
   }
   FortConversationRequirement_ControllerRequirement_13 {
    class=FortConversationRequirement_ControllerRequirement
    properties {
     nodeName=Require: Has GameplayTag
     requirement {class=FortControllerRequirement_HasOwnedTag, properties {bInverseCondition=true, ownedTag {tagName=Granted.Athena.UrgentQuest.Active}}}
    }
   }
  ]
 }
 GrantSlottedQuest_C_4 {
  id=#19
  class=GrantSlottedQuest_C
  outputs []
  properties {
   bSharedQuestLimitsEnforced=false
   bShouldOverrideSpeakerText=false
   briefWidgetClassOverride=/JobBoard_BR_S37/Widgets/FortSupplyDropJackpotJobBrief.FortSupplyDropJackpotJobBrief_C
   choiceSound=null
   softTaskIcon=&T-Icon-Check-128
   optionalFilterCategoryTag {tagName=Tandem.Service.JobType.S37.SupplyJackpot}
  }
  sub_nodes [
   ConversationChoiceNode_0 {class=ConversationChoiceNode, properties {defaultChoiceDisplayText=SUPPLY JACKPOT BRIEFING}}
   SideEffect_PlayServiceUsedEffects_C_0 {class=SideEffect_PlayServiceUsedEffects_C}
   SideEffect_Server_RecordEventToHistoryData_C_0 {class=SideEffect_Server_RecordEventToHistoryData_C, properties {event Type {tagName=Conversation.Event.GrantedQuest}}}
   SideEffect_Server_TrackJobBoardUser_C_0 {class=SideEffect_Server_TrackJobBoardUser_C}
   SideEffect_Server_JobBoardBriefing_SpawnActorOfClass_C_0 {
    class=SideEffect_Server_JobBoardBriefing_SpawnActorOfClass_C
    properties {spawnerClass=/JobBoard_BR_S37/Blueprints/Shared/B_Athena_JobBoardBriefing_Spawner_SupplyDrop_S37_Jackpot.B_Athena_JobBoardBriefing_Spawner_SupplyDrop_S37_Jackpot_C}
   }
   SideEffect_Server_JobBoardBriefing_ApplyInstigatorGE_C_0 {class=SideEffect_Server_JobBoardBriefing_ApplyInstigatorGE_C}
   SideEffect_Server_JobBoardBriefing_DeactivateOnUse_C_0 {class=SideEffect_Server_JobBoardBriefing_DeactivateOnUse_C}
   FortConversationRequirement_ControllerRequirement_0 {
    class=FortConversationRequirement_ControllerRequirement
    properties {
     nodeName=Require: Has Recorded Conversation Event With NPC
     requirement {
      class=FortControllerRequirement_HasRecordedConversationEventWithNPC
      properties {bInverseCondition=true, queryMethod=CheckAgainstCurrentConversationParticipant, queryEventTag {tagName=Conversation.Event.GrantedQuest}}
     }
    }
   }
   FortConversationRequirement_ControllerRequirement_1 {
    class=FortConversationRequirement_ControllerRequirement
    properties {
     nodeName=Require: Has Urgent Quest
     requirement {
      class=FortControllerRequirement_HasUrgentQuest
      properties {
       bInverseCondition=true
       bRequireExactMatch=false
       acceptedUrgentEventTags {gameplayTags [{tagName=SpecialEvent.HuntingPlayer.IsHunter}, {tagName=SpecialEvent.JobBoard.Experimental}]}
      }
     }
    }
   }
   FortConversationRequirement_HasService_0 {class=FortConversationRequirement_HasService, properties {serviceTag {tagName=Tandem.Service.JobType.S37.SupplyJackpot}}}
   FortConversationRequirement_ControllerRequirement_2 {
    class=FortConversationRequirement_ControllerRequirement
    properties {
     nodeName=Require: Has GameplayTag
     requirement {
      class=FortControllerRequirement_HasOwnedTag
      properties {bInverseCondition=true, ownedTag {tagName=Granted.Athena.JobBoardExperimental.QuestActive}}
     }
    }
   }
   FortConversationRequirement_ControllerRequirement_3 {
    class=FortConversationRequirement_ControllerRequirement
    properties {
     nodeName=Require: Has GameplayTag
     requirement {class=FortControllerRequirement_HasOwnedTag, properties {bInverseCondition=true, ownedTag {tagName=Granted.Athena.BountyBoard.QuestActive}}}
    }
   }
   FortConversationRequirement_ControllerRequirement_12 {
    class=FortConversationRequirement_ControllerRequirement
    properties {
     nodeName=Require: Has GameplayTag
     requirement {class=FortControllerRequirement_HasOwnedTag, properties {bInverseCondition=true, ownedTag {tagName=Granted.Athena.UrgentQuest.Active}}}
    }
   }
  ]
 }
]
